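# Physics asset for the male medium Parcel Prank body cosmetic, covering secondary motion of its ribbon and waistcoat bones.
# Capsules on pelvis, spine, thighs and feet form a simplified body for the cloth pieces.
# Small master capsules attach to the spine with free rotation, anchoring two-link box ribbon chains and double-box waistcoat flaps on each side.
# Ribbon and flap joints restrict twist tightly, and the flaps swing much further on one axis.
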PhysicsAsset /BRCosmetics/Characters/Player/Male/Medium/Bodies/M_MED_Parcel_Prank/Meshes/M_MED_Parcel_Prank_dyn
bodies {
 dyn_ribbon_MSTR [{shapeType=Capsule, length=0.5, radius=0.5}]
 dyn_ribbon_l_01 [{shapeType=Box, extentX=4, extentY=2.5, extentZ=1.5, center=<2 0 0>}]
 dyn_ribbon_l_02 [{shapeType=Box, extentX=4, extentY=2.5, extentZ=1.5, center=<3 0 0>}]
 dyn_ribbon_r_01 [{shapeType=Box, extentX=4, extentY=2.5, extentZ=1.5, center=<2 0 0>}]
 dyn_ribbon_r_02 [{shapeType=Box, extentX=4, extentY=2.5, extentZ=1.5, center=<3 0 0>}]
 dyn_waistcoat_l [
  {shapeType=Box, extentX=8.5, extentY=4.5, extentZ=1, center=<5.7418 2.8328 -0.2513>, rotation=<rot 1.9267 -29.9089 0.2229>}
  {shapeType=Box, extentX=7.5, extentY=4.5, extentZ=1, center=<3.5404 -1.0194 -0.5758>, rotation=<rot 1.5922 -29.002 -8.6217>}
 ]
 dyn_waistcoat_l_MSTR [{shapeType=Capsule, length=0.5, radius=0.5}]
 dyn_waistcoat_r [
  {shapeType=Box, extentX=8, extentY=4.5, extentZ=1, center=<5.7418 -2.8328 -0.2513>, rotation=<rot 1.9267 29.9089 -0.2229>}
  {shapeType=Box, extentX=7, extentY=4.5, extentZ=1, center=<3.628 0.8598 -0.5484>, rotation=<rot 1.5922 29.0019 8.6213>}
 ]
 dyn_waistcoat_r_MSTR [{shapeType=Capsule, length=0.5, radius=0.5}]
 foot_l [{shapeType=Capsule, length=2, radius=2}]
 foot_r [{shapeType=Capsule, length=2, radius=2}]
 pelvis [{shapeType=Capsule, length=9.5, radius=9.25}]
 spine_01 [{shapeType=Capsule, length=11, radius=7, center=<4.7862 -0.8239 0.0001>}]
 spine_02 [{shapeType=Capsule, length=11.25, radius=6.75, center=<1.0737 -0.5893 0>}]
 spine_03 [{shapeType=Capsule, length=10.75, radius=7.25, center=<0 -0.0119 0>}]
 thigh_l [{shapeType=Capsule, length=28.25, radius=9.25, center=<-16.9515 -1.0787 -0.5152>, rotation=<rot 85.5817 175.6316 169.6804>}]
 thigh_r [{shapeType=Capsule, length=31, radius=9.25, center=<18.7625 1.2422 0.7789>, rotation=<rot 85.577 175.6319 169.6288>}]
}
constraints [
 {bone1Name=spine_01, bone2Name=pelvis, swing1LimitDegrees=45, swing2LimitDegrees=45, twistLimitDegrees=45}
 {bone1Name=spine_03, bone2Name=spine_02, swing1LimitDegrees=45, swing2LimitDegrees=45, twistLimitDegrees=45}
 {bone1Name=foot_l, bone2Name=thigh_l, swing1LimitDegrees=45, swing2LimitDegrees=45, twistLimitDegrees=45}
 {bone1Name=foot_r, bone2Name=thigh_r, swing1LimitDegrees=45, swing2LimitDegrees=45, twistLimitDegrees=45}
 {
  bone1Name=dyn_ribbon_MSTR
  bone2Name=spine_03
  swing1LimitDegrees=45
  swing1Motion=Free
  swing2LimitDegrees=45
  swing2Motion=Free
  twistLimitDegrees=45
  twistMotion=Free
 }
 {bone1Name=dyn_ribbon_l_01, bone2Name=dyn_ribbon_MSTR, swing1LimitDegrees=45, swing2LimitDegrees=45, twistLimitDegrees=5}
 {bone1Name=dyn_ribbon_l_02, bone2Name=dyn_ribbon_l_01, swing1LimitDegrees=45, swing2LimitDegrees=45, twistLimitDegrees=5}
 {bone1Name=dyn_ribbon_r_01, bone2Name=dyn_ribbon_MSTR, swing1LimitDegrees=45, swing2LimitDegrees=45, twistLimitDegrees=5}
 {bone1Name=dyn_ribbon_r_02, bone2Name=dyn_ribbon_r_01, swing1LimitDegrees=45, swing2LimitDegrees=45, twistLimitDegrees=5}
 {
  bone1Name=spine_02
  bone2Name=spine_01
  swing1LimitDegrees=45
  swing1Motion=Free
  swing2LimitDegrees=45
  swing2Motion=Free
  twistLimitDegrees=45
  twistMotion=Free
 }
 {
  bone1Name=dyn_waistcoat_l_MSTR
  bone2Name=spine_01
  swing1LimitDegrees=45
  swing1Motion=Free
  swing2LimitDegrees=45
  swing2Motion=Free
  twistLimitDegrees=45
  twistMotion=Free
 }
 {bone1Name=dyn_waistcoat_l, bone2Name=dyn_waistcoat_l_MSTR, swing1LimitDegrees=10, swing2LimitDegrees=60, twistLimitDegrees=5}
 {
  bone1Name=dyn_waistcoat_r_MSTR
  bone2Name=spine_01
  swing1LimitDegrees=45
  swing1Motion=Free
  swing2LimitDegrees=45
  swing2Motion=Free
  twistLimitDegrees=45
  twistMotion=Free
 }
 {bone1Name=dyn_waistcoat_r, bone2Name=dyn_waistcoat_r_MSTR, swing1LimitDegrees=10, swing2LimitDegrees=60, twistLimitDegrees=5}
 {
  bone1Name=thigh_l
  bone2Name=pelvis
  swing1LimitDegrees=45
  swing1Motion=Free
  swing2LimitDegrees=45
  swing2Motion=Free
  twistLimitDegrees=45
  twistMotion=Free
 }
 {
  bone1Name=thigh_r
  bone2Name=pelvis
  swing1LimitDegrees=45
  swing1Motion=Free
  swing2LimitDegrees=45
  swing2Motion=Free
  twistLimitDegrees=45
  twistMotion=Free
 }
]
