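# Physics asset for the dynamic chain on the Parcel Prank male medium body cosmetic.
# Two capsule chains, A with three links and B with four, branch from small master bodies constrained to the pelvis with free rotation.
# Per-link swing and twist limits tighten at the tips, where an angularly locked constraint joins both ends into a closed loop.
# Larger pelvis and right thigh capsules stand in for the character's body.
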PhysicsAsset /BRCosmetics/Characters/Player/Male/Medium/Bodies/M_MED_Parcel_Prank/Meshes/M_MED_Parcel_Prank_dyn_chain
bodies {
 dyn_chain_a_01 [{shapeType=Capsule, length=5.5449, radius=1.2, center=<4.2656 0 0.1783>, rotation=<rot 0 95.6845 -87.1423>}]
 dyn_chain_a_02 [{shapeType=Capsule, length=5.9242, radius=1.2, center=<3.7352 -0.1254 -0.1446>, rotation=<rot 0 87.2355 -88.2084>}]
 dyn_chain_a_03 [{shapeType=Capsule, length=5.0927, radius=1.45, center=<2.0548 -0.3309 0.076>, rotation=<rot -2.4305 88.4551 -85.7836>}]
 dyn_chain_a_MSTR [{shapeType=Capsule, length=0.5, radius=0.5}]
 dyn_chain_b_01 [{shapeType=Capsule, length=2.7378, radius=1.2, center=<4.4432 -0.1183 0.0196>, rotation=<rot 0 89.2155 -89.0017>}]
 dyn_chain_b_02 [{shapeType=Capsule, length=3.4641, radius=1.2, center=<4.3333 -0.0749 0.0228>, rotation=<rot 0 90.5523 -89.7942>}]
 dyn_chain_b_03 [{shapeType=Capsule, length=3.2863, radius=1.2, center=<3.7964 -0.3638 0.31>, rotation=<rot 0.0307 -88.2167 -89.6491>}]
 dyn_chain_b_04 [{shapeType=Capsule, length=5.3668, radius=1.45, center=<2.3574 -0.2074 -0.3882>, rotation=<rot -2.7713 -83.06 -79.5904>}]
 dyn_chain_b_MSTR [{shapeType=Capsule, length=0.5, radius=0.5}]
 pelvis [{shapeType=Capsule, length=12, radius=8, center=<0 -0.1309 0>}]
 thigh_twist_01_r [{shapeType=Capsule, length=15.5, radius=7.75, center=<-9.5945 -0.1759 -0.9139>, rotation=<rot 88.2015 4.9949 0.7572>}]
}
constraints [
 {
  bone1Name=thigh_twist_01_r
  bone2Name=pelvis
  swing1LimitDegrees=45
  swing1Motion=Free
  swing2LimitDegrees=45
  swing2Motion=Free
  twistLimitDegrees=45
  twistMotion=Free
 }
 {
  bone1Name=dyn_chain_a_MSTR
  bone2Name=pelvis
  swing1LimitDegrees=45
  swing1Motion=Free
  swing2LimitDegrees=45
  swing2Motion=Free
  twistLimitDegrees=45
  twistMotion=Free
 }
 {bone1Name=dyn_chain_a_01, bone2Name=dyn_chain_a_MSTR, swing1LimitDegrees=20, swing2LimitDegrees=30, twistLimitDegrees=5}
 {bone1Name=dyn_chain_a_02, bone2Name=dyn_chain_a_01, swing1LimitDegrees=30, swing2LimitDegrees=30, twistLimitDegrees=5}
 {bone1Name=dyn_chain_a_03, bone2Name=dyn_chain_a_02, swing1LimitDegrees=10, swing2LimitDegrees=10, twistLimitDegrees=2}
 {
  bone1Name=dyn_chain_b_MSTR
  bone2Name=pelvis
  swing1LimitDegrees=45
  swing1Motion=Free
  swing2LimitDegrees=45
  swing2Motion=Free
  twistLimitDegrees=45
  twistMotion=Free
 }
 {bone1Name=dyn_chain_b_01, bone2Name=dyn_chain_b_MSTR, swing1LimitDegrees=60, swing2LimitDegrees=5, twistLimitDegrees=5}
 {bone1Name=dyn_chain_b_02, bone2Name=dyn_chain_b_01, swing1LimitDegrees=60, swing2LimitDegrees=60, twistLimitDegrees=5}
 {bone1Name=dyn_chain_b_03, bone2Name=dyn_chain_b_02, swing1LimitDegrees=60, swing2LimitDegrees=60, twistLimitDegrees=5}
 {bone1Name=dyn_chain_b_04, bone2Name=dyn_chain_b_03, swing1LimitDegrees=10, swing2LimitDegrees=10, twistLimitDegrees=2}
 {
  bone1Name=dyn_chain_a_03
  bone2Name=dyn_chain_b_04
  swing1LimitDegrees=15
  swing1Motion=Locked
  swing2LimitDegrees=15
  swing2Motion=Locked
  twistLimitDegrees=5
  twistMotion=Locked
 }
]
