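# Physics asset for the left shoelace of the male medium Parcel Prank character body, defining simulation shapes for its dynamic bones.
# Two small box bodies cover the shoelace bones, and a capsule represents the left foot.
# Each lace bone is constrained directly to the foot rather than chained to the other, with moderate swing limits and very tight twist.
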PhysicsAsset /BRCosmetics/Characters/Player/Male/Medium/Bodies/M_MED_Parcel_Prank/Meshes/M_MED_Parcel_Prank_dyn_shoelace_l
bodies {
 dyn_shoelace_01_l [{shapeType=Box, extentX=2, extentY=2, extentZ=1.2, center=<2 0 0>}]
 dyn_shoelace_02_l [{shapeType=Box, extentX=2, extentY=2, extentZ=1.2, center=<2.9669 0 0>}]
 foot_l [{shapeType=Capsule, length=8.2, radius=3.8, center=<-4.4716 4.6391 -0.4315>, rotation=<rot 90 14.0363 -93.5145>}]
}
constraints [
 {bone1Name=dyn_shoelace_01_l, bone2Name=foot_l, swing1LimitDegrees=30, swing2LimitDegrees=45, twistLimitDegrees=2}
 {bone1Name=dyn_shoelace_02_l, bone2Name=foot_l, swing1LimitDegrees=30, swing2LimitDegrees=45, twistLimitDegrees=2}
]
