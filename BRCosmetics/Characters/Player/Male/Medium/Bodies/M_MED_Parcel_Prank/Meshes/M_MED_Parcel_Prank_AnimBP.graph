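# Post-process style secondary-motion AnimBP for the Parcel Prank cosmetic body, layered on the pose copied from its attached parent mesh.
# Thigh lift, measured each update, drives additive waistcoat and chain bone offsets after a spine/pelvis-weighted ribbon constraint.
# AnimDynamics then swings the chain, followed by RigidBody sims for the body and each shoelace; a chain RigidBody node sits unconnected.
# Crouch, skydive and delayed parachute events reset dynamics.
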AnimBlueprint /BRCosmetics/Characters/Player/Male/Medium/Bodies/M_MED_Parcel_Prank/Meshes/M_MED_Parcel_Prank_AnimBP
target_skeleton=&/M_MED_Parcel_Prank_Skeleton
graphs [
 AnimGraph {nodes [
   RigidBody() {
    bClampLinearTranslationLimitToRefPose=true
    bOverrideWorldGravity=true
    overridePhysicsAsset=&/M_MED_Parcel_Prank_dyn_chain
    simulationSpace=BaseBoneSpace
    alphaBoolBlend {blendOption=Linear}
    baseBoneRef {boneName=pelvis}
    componentAppliedLinearAccClamp=<500 500 500>
    componentLinearAccScale=<0.15 0.15 0.15>
    componentLinearVelScale=<0.3 0.3 0.3>
    overrideWorldGravity {z=-980}
   }
   n1 = "Constraint - dyn_ribbon_MSTR (PPPPPP)"!(
    ComponentPose=Local To Component(LocalPose=Copy Pose From Mesh() {bUseAttachedParent=true})
    Alpha=1
    ConstraintWeights_0=0.1623
    ConstraintWeights_1:float=0.2953
    ConstraintWeights_2:float=0.2696
    ConstraintWeights_3:float=0.133
    ConstraintWeights_4:float=0.0634
    ConstraintWeights_5:float=0.0764
   ) -> (Pose) {
    alphaBoolBlend {blendOption=Linear}
    boneToModify {boneName=dyn_ribbon_MSTR}
    constraintSetup [
     {targetBone {boneName=spine_01}}
     {targetBone {boneName=spine_02}}
     {targetBone {boneName=spine_03}}
     {targetBone {boneName=spine_04}}
     {targetBone {boneName=spine_05}}
     {targetBone {boneName=pelvis}}
    ]
    constraintWeights [1,1,1,1,1,1]
   }
   n2 = "Transform (Modify) Bone - Bone: dyn_waistcoat_l_MSTR"(ComponentPose=@n1, Alpha=Get thigh_rot_up_l() -> (thigh_rot_up_l:double), Translation="-5,0,2.5", Rotation="0,0,-40", Scale="1,1,1") {
    rotationMode=BMM_Additive
    rotationSpace=BCS_BoneSpace
    translationMode=BMM_Additive
    translationSpace=BCS_BoneSpace
    alphaBoolBlend {blendOption=Linear}
    boneToModify {boneName=dyn_waistcoat_l_MSTR}
    rotation {roll=-40}
    translation {x=-5, z=2.5}
   }
   n3 = Get thigh_rot_up_r() -> (thigh_rot_up_r:double)
   n4 = "Transform (Modify) Bone - Bone: dyn_waistcoat_r_MSTR"(ComponentPose=@n2, Alpha=@n3, Translation="-5,0,2.5", Rotation="0,0,40", Scale="1,1,1") {
    rotationMode=BMM_Additive
    rotationSpace=BCS_BoneSpace
    translationMode=BMM_Additive
    translationSpace=BCS_BoneSpace
    alphaBoolBlend {blendOption=Linear}
    boneToModify {boneName=dyn_waistcoat_r_MSTR}
    rotation {roll=40}
    translation {x=-5, z=2.5}
   }
   n5 = "Transform (Modify) Bone - Bone: dyn_chain_b_MSTR"(ComponentPose=@n4, Alpha=@n3, Translation="-0.837,3.257,-2.337", Rotation="0.000,0.000,30", Scale="1,1,1") {
    rotationMode=BMM_Additive
    rotationSpace=BCS_BoneSpace
    translationMode=BMM_Additive
    translationSpace=BCS_BoneSpace
    alphaBoolBlend {blendOption=Linear}
    boneToModify {boneName=dyn_chain_b_MSTR}
    rotation {roll=30}
    translation=<-0.837 3.257 -2.337>
   }
   n6 = Anim Dynamics - Bone: dyn_chain(ComponentPose=@n5, Alpha=1, ExternalForce="0,0,0") {
    angularDampingOverride=1
    angularSpringConstant=1500
    bAngularSpring=true
    bGravityOverrideInSimSpace=true
    bOverrideAngularDamping=true
    gravityScale=0.35
    simulationSpace=RootRelative
    alphaBoolBlend {blendOption=Linear}
    alphaScaleBiasClamp {scale=0.6}
    boundBone {boneName=dyn_chain}
    componentAppliedLinearAccClamp {z=5000}
    componentLinearAccScale {z=0.45}
    componentLinearVelScale {z=0.45}
    physicsBodyDefinitions [{
      sphereCollisionRadius=0
      boundBone {boneName=dyn_chain}
      boxExtents=<10 20 20>
      constraintSetup {angularLimitsMax=<0 40 0>, angularLimitsMin=<0 0 0>, angularTarget=<0 0 0>}
      localJointOffset=<10 0 0>
     }]
   }
   n7 = RigidBody(ComponentPose=@n6) {
    bClampLinearTranslationLimitToRefPose=true
    bOverrideWorldGravity=true
    overridePhysicsAsset=&/M_MED_Parcel_Prank_dyn
    simulationSpace=BaseBoneSpace
    alphaBoolBlend {blendOption=Linear}
    baseBoneRef {boneName=pelvis}
    componentAppliedLinearAccClamp=<2000 2000 2000>
    componentLinearAccScale=<0.3 0.3 0.3>
    componentLinearVelScale=<0.3 0.3 0.3>
    overrideWorldGravity {z=-980}
   }
   n8 = RigidBody(ComponentPose=@n7) {
    bClampLinearTranslationLimitToRefPose=true
    bOverrideWorldGravity=true
    overridePhysicsAsset=&/M_MED_Parcel_Prank_dyn_shoelace_r
    simulationSpace=BaseBoneSpace
    alphaBoolBlend {blendOption=Linear}
    baseBoneRef {boneName=calf_r}
    componentAppliedLinearAccClamp=<1000 1000 1000>
    componentLinearAccScale=<0.25 0.25 0.25>
    componentLinearVelScale=<0.2 0.2 0.2>
    overrideWorldGravity {z=-500}
   }
   n9 = RigidBody(ComponentPose=@n8) {
    bClampLinearTranslationLimitToRefPose=true
    bOverrideWorldGravity=true
    overridePhysicsAsset=&/M_MED_Parcel_Prank_dyn_shoelace_l
    simulationSpace=BaseBoneSpace
    alphaBoolBlend {blendOption=Linear}
    baseBoneRef {boneName=calf_l}
    componentAppliedLinearAccClamp=<1000 1000 1000>
    componentLinearAccScale=<0.25 0.25 0.25>
    componentLinearVelScale=<0.2 0.2 0.2>
    overrideWorldGravity {z=-500}
   }
   Output Pose(Result=Component To Local(ComponentPose=@n9)) {layerGroup=None}
  ]}
 EventGraph {nodes [
   Event BlueprintUpdateAnimation()
   n1 = BreakTransform(InTransform=GetDeltaTransformFromRefPose(self=GetOwnerSkeletalMeshComponent(), BoneName=thigh_r, BaseName=pelvis))
   > Set thigh_rot_up_r(thigh_rot_up_r:double=MapRangeClamped(Value=BreakRotator(InRot=@n1.Rotation).Yaw, InRangeA=0, InRangeB=-120)) -> (Output_Get:double)
   n2 = BreakTransform(InTransform=GetDeltaTransformFromRefPose(self=GetOwnerSkeletalMeshComponent(), BoneName=thigh_l, BaseName=pelvis))
   > Set thigh_rot_up_l(thigh_rot_up_l:double=MapRangeClamped(Value=BreakRotator(InRot=@n2.Rotation).Yaw, InRangeA=-30, InRangeB=-120)) -> (Output_Get:double)
   n3 = Event OnBeginCrouching()
   ResetDynamics(execute=[@n4.then, @n3.then, @n5.then, @n6.then], InTeleportType=TeleportPhysics)
   > ResetDynamics(InTeleportType=ResetPhysics)
   Event OnParachuteOpened()
   > n4 = Delay(Duration=1.5)
   n5 = Event OnEndSkydiving()
   n6 = Event OnBeginSkydiving()
  ]}
]
variables [
 thigh_rot_up_l {type=number}
 thigh_rot_up_r {type=number}
]
